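# Physics asset for the DaggerPack backpack cosmetic, giving collision and joint setup to the pack and its skull piece.
# A capsule body on Master_Skull_jnt covers the skull, and a much smaller, near-spherical capsule sits on the pack bone.
# A single constraint links the skull to the pack, with narrow swing limit angles and a zero twist limit angle.
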PhysicsAsset /BRCosmetics/Accessories/FORT_Backpacks/Backpack_M_MED_DaggerPack/Meshes/M_MED_DaggerPack_Pack_Physics
bodies {
 Master_Skull_jnt [{shapeType=Capsule, length=0, radius=5, shapeName=Master_Skull_jnt_capsule, center=<0 -5.686 0>, rotation=<rot 0 -90 90>}]
 pack [{shapeType=Capsule, length=0.505, radius=0.505, shapeName=pack_capsule, rotation=<rot 0 -90 0>}]
}
constraints [{bone1Name=Master_Skull_jnt, bone2Name=pack, swing1LimitDegrees=10, swing2LimitDegrees=10, twistLimitDegrees=0}]
