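# Animation Blueprint for the DaggerPack backpack cosmetic, adding procedural float and physics to its skull and needle bones.
# Each update builds three sine-driven offsets, applied additively to needle groups for desynchronized bobbing.
# Faint spring controllers precede the offsets and switch off when pawn speed exceeds a cutoff.
# A final RigidBody node simulates relative to the pack bone, using gravity from a deprecated override function; the Crown spring node is unconnected.
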AnimBlueprint /BRCosmetics/Accessories/FORT_Backpacks/Backpack_M_MED_DaggerPack/Meshes/M_MED_DaggerPack_Pack_AnimBP
target_skeleton=&/M_MED_DaggerPack_Pack_Skeleton
graphs [
 AnimGraph {nodes [
   n1 = Local To Component(LocalPose=Input Pose())
   n2 = Get SpringAlpha() -> (SpringAlpha:double)
   n3 = Spring controller - Bone: Rt_Needle_02_jnt!(
    ComponentPose:Component Space Pose Link Structure=Spring controller - Bone: Rt_Needle_01_jnt!(
     ComponentPose:Component Space Pose Link Structure=Spring controller - Bone: Skull_jnt!(ComponentPose:Component Space Pose Link Structure=@n1, Alpha:float=@n2) -> (Pose:Component Space Pose Link Structure) {
      is=title AnimGraphNode_SpringBone
      springDamping=2.5
      springStiffness=150
      alphaBoolBlend {blendOption=Linear}
      springBone {boneName=Skull_jnt}
     }
     Alpha:float=@n2
    ) -> (Pose:Component Space Pose Link Structure) {
     is=title AnimGraphNode_SpringBone
     springDamping=2.5
     springStiffness=130
     alphaBoolBlend {blendOption=Linear}
     springBone {boneName=Rt_Needle_01_jnt}
    }
    Alpha:float=@n2
   ) -> (Pose:Component Space Pose Link Structure) {
    is=title AnimGraphNode_SpringBone
    springDamping=2.5
    springStiffness=125
    alphaBoolBlend {blendOption=Linear}
    springBone {boneName=Rt_Needle_02_jnt}
   }
   n4 = Spring controller - Bone: Rt_Needle_05_jnt!(
    ComponentPose:Component Space Pose Link Structure=Spring controller - Bone: Rt_Needle_04_jnt!(
     ComponentPose:Component Space Pose Link Structure=Spring controller - Bone: Rt_Needle_03_jnt!(ComponentPose:Component Space Pose Link Structure=@n3, Alpha:float=@n2) -> (Pose:Component Space Pose Link Structure) {
      is=title AnimGraphNode_SpringBone
      springDamping=2.5
      springStiffness=120
      alphaBoolBlend {blendOption=Linear}
      springBone {boneName=Rt_Needle_03_jnt}
     }
     Alpha:float=@n2
    ) -> (Pose:Component Space Pose Link Structure) {
     is=title AnimGraphNode_SpringBone
     springDamping=2.5
     springStiffness=115
     alphaBoolBlend {blendOption=Linear}
     springBone {boneName=Rt_Needle_04_jnt}
    }
    Alpha:float=@n2
   ) -> (Pose:Component Space Pose Link Structure) {
    is=title AnimGraphNode_SpringBone
    springDamping=2.5
    springStiffness=110
    alphaBoolBlend {blendOption=Linear}
    springBone {boneName=Rt_Needle_05_jnt}
   }
   n5 = Spring controller - Bone: Lf_Needle_03_jnt!(
    ComponentPose:Component Space Pose Link Structure=Spring controller - Bone: Lf_Needle_02_jnt!(
     ComponentPose:Component Space Pose Link Structure=Spring controller - Bone: Lf_Needle_01_jnt!(ComponentPose:Component Space Pose Link Structure=@n4, Alpha:float=@n2) -> (Pose:Component Space Pose Link Structure) {
      is=title AnimGraphNode_SpringBone
      springDamping=2.5
      springStiffness=130
      alphaBoolBlend {blendOption=Linear}
      springBone {boneName=Lf_Needle_01_jnt}
     }
     Alpha:float=@n2
    ) -> (Pose:Component Space Pose Link Structure) {
     is=title AnimGraphNode_SpringBone
     springDamping=2.5
     springStiffness=125
     alphaBoolBlend {blendOption=Linear}
     springBone {boneName=Lf_Needle_02_jnt}
    }
    Alpha:float=@n2
   ) -> (Pose:Component Space Pose Link Structure) {
    is=title AnimGraphNode_SpringBone
    springDamping=2.5
    springStiffness=120
    alphaBoolBlend {blendOption=Linear}
    springBone {boneName=Lf_Needle_03_jnt}
   }
   n6 = Get Crown_Float_01() -> (Crown_Float_01:Vector)
   n7 = "Transform (Modify) Bone - Bone: Rt_Needle_01_jnt"(
    ComponentPose=Spring controller - Bone: Ct_Needle_01_jnt!(
     ComponentPose:Component Space Pose Link Structure=Spring controller - Bone: Lf_Needle_04_jnt!(ComponentPose:Component Space Pose Link Structure=@n5, Alpha:float=@n2) -> (Pose:Component Space Pose Link Structure) {
      is=title AnimGraphNode_SpringBone
      springDamping=2.5
      springStiffness=115
      alphaBoolBlend {blendOption=Linear}
      springBone {boneName=Lf_Needle_04_jnt}
     }
     Alpha:float=@n2
    ) -> (Pose:Component Space Pose Link Structure) {
     is=title AnimGraphNode_SpringBone
     springDamping=2.5
     springStiffness=135
     alphaBoolBlend {blendOption=Linear}
     springBone {boneName=Ct_Needle_01_jnt}
    }
    Alpha=1
    Translation=@n6
    Rotation="0,0,0"
    Scale="1,1,1"
   ) {translationMode=BMM_Additive, translationSpace=BCS_BoneSpace, alphaBoolBlend {blendOption=Linear}, boneToModify {boneName=Rt_Needle_01_jnt}}
   n8 = "Transform (Modify) Bone - Bone: Rt_Needle_03_jnt"(ComponentPose=@n7, Alpha=1, Translation=@n6, Rotation="0,0,0", Scale="1,1,1") {translationMode=BMM_Additive, translationSpace=BCS_BoneSpace, alphaBoolBlend {blendOption=Linear}, boneToModify {boneName=Rt_Needle_03_jnt}}
   n9 = "Transform (Modify) Bone - Bone: Lf_Needle_01_jnt"(ComponentPose=@n8, Alpha=1, Translation=@n6, Rotation="0,0,0", Scale="1,1,1") {translationMode=BMM_Additive, translationSpace=BCS_BoneSpace, alphaBoolBlend {blendOption=Linear}, boneToModify {boneName=Lf_Needle_01_jnt}}
   n10 = "Transform (Modify) Bone - Bone: Rt_Needle_05_jnt"(ComponentPose=@n9, Alpha=1, Translation=@n6, Rotation="0,0,0", Scale="1,1,1") {translationMode=BMM_Additive, translationSpace=BCS_BoneSpace, alphaBoolBlend {blendOption=Linear}, boneToModify {boneName=Rt_Needle_05_jnt}}
   n11 = "Transform (Modify) Bone - Bone: Lf_Needle_03_jnt"(ComponentPose=@n10, Alpha=1, Translation=@n6, Rotation="0,0,0", Scale="1,1,1") {translationMode=BMM_Additive, translationSpace=BCS_BoneSpace, alphaBoolBlend {blendOption=Linear}, boneToModify {boneName=Lf_Needle_03_jnt}}
   n12 = Get Crown_Float_02() -> (Crown_Float_02:Vector)
   n13 = "Transform (Modify) Bone - Bone: Ct_Needle_01_jnt"(ComponentPose=@n11, Alpha=1, Translation=@n12, Rotation="0,0,0", Scale="1,1,1") {translationMode=BMM_Additive, translationSpace=BCS_BoneSpace, alphaBoolBlend {blendOption=Linear}, boneToModify {boneName=Ct_Needle_01_jnt}}
   n14 = "Transform (Modify) Bone - Bone: Rt_Needle_02_jnt"(ComponentPose=@n13, Alpha=1, Translation=@n12, Rotation="0,0,0", Scale="1,1,1") {translationMode=BMM_Additive, translationSpace=BCS_BoneSpace, alphaBoolBlend {blendOption=Linear}, boneToModify {boneName=Rt_Needle_02_jnt}}
   n15 = "Transform (Modify) Bone - Bone: Lf_Needle_02_jnt"(ComponentPose=@n14, Alpha=1, Translation=@n12, Rotation="0,0,0", Scale="1,1,1") {translationMode=BMM_Additive, translationSpace=BCS_BoneSpace, alphaBoolBlend {blendOption=Linear}, boneToModify {boneName=Lf_Needle_02_jnt}}
   n16 = Get Crown_Float_03() -> (Crown_Float_03:Vector)
   n17 = "Transform (Modify) Bone - Bone: Rt_Needle_04_jnt"(ComponentPose=@n15, Alpha=1, Translation=@n16, Rotation="0,0,0", Scale="1,1,1") {translationMode=BMM_Additive, translationSpace=BCS_BoneSpace, alphaBoolBlend {blendOption=Linear}, boneToModify {boneName=Rt_Needle_04_jnt}}
   n18 = "Transform (Modify) Bone - Bone: Lf_Needle_04_jnt"(ComponentPose=@n17, Alpha=1, Translation=@n16, Rotation="0,0,0", Scale="1,1,1") {translationMode=BMM_Additive, translationSpace=BCS_BoneSpace, alphaBoolBlend {blendOption=Linear}, boneToModify {boneName=Lf_Needle_04_jnt}}
   n19 = RigidBody!(ComponentPose=@n18, Alpha, OverrideWorldGravity:Vector=Get GravityOverride(), ExternalForce) -> (Pose) {
    bClampLinearTranslationLimitToRefPose=true
    bOverrideWorldGravity=true
    overridePhysicsAsset=&/M_MED_DaggerPack_Pack_Physics
    simulationSpace=BaseBoneSpace
    alphaBoolBlend {blendOption=Linear}
    baseBoneRef {boneName=pack}
    componentAppliedLinearAccClamp=<3000 3000 3000>
    componentLinearAccScale=<0.4 0.3 0.3>
    componentLinearVelScale=<0.4 0.3 0.3>
    simSpaceSettings {
     maxAngularAcceleration=1000
     maxAngularVelocity=3
     maxLinearAcceleration=700
     maxLinearVelocity=500
     velocityScaleZ=0
     worldAlpha=1
     externalLinearDragV=<0.2 0.2>
    }
   }
   Output Pose(Result=Component To Local(ComponentPose=@n19)) {layerGroup=None}
   Spring controller - Bone: Crown_jnt!(ComponentPose:Component Space Pose Link Structure, Alpha:float=0.01) -> (Pose:Component Space Pose Link Structure) {
    is=title AnimGraphNode_SpringBone
    springDamping=2.5
    springStiffness=140
    alphaBoolBlend {blendOption=Linear}
    springBone {boneName=Crown_jnt}
   }
  ]}
 EventGraph {nodes [
   Event BlueprintUpdateAnimation()
   > n1 = Sequence()
   n2 = GetRealTimeSeconds()
   n3 = "Sin (Radians)"(A=float * float(A=float * float(A=0.5, B=0.5), B=@n2))
   n4 = "Sin (Radians)"(A=float * float(A=float * float(A=0.5, B=0.5), B=@n2))
   n5 = "Sin (Radians)"(A=float * float(A=float * float(A=1.7, B=1.7), B=@n2))
   > Set Crown_Float_01(Crown_Float_01:Vector=MakeVector(X=float * float(A=@n5, B=0.7), Y=float * float(A=@n4, B=0.7), Z=float * float(A=@n3, B=0.7))) -> (Output_Get:Vector)
   n6 = "Sin (Radians)"(A=float * float(A=float * float(A=0.5, B=0.5), B=@n2))
   n7 = "Sin (Radians)"(A=float * float(A=float * float(A=0.6, B=0.6), B=@n2))
   n8 = "Sin (Radians)"(A=float * float(A=float * float(A=1.5, B=1.5), B=@n2))
   > Set Crown_Float_02(Crown_Float_02:Vector=MakeVector(X=float * float(A=@n8, B=0.7), Y=float * float(A=@n7, B=0.7), Z=float * float(A=@n6, B=0.7))) -> (Output_Get:Vector)
   n9 = "Sin (Radians)"(A=float * float(A=float * float(A=0.4, B=0.4), B=@n2))
   n10 = "Sin (Radians)"(A=float * float(A=float * float(A=1.9, B=1.9), B=@n2))
   n11 = "Sin (Radians)"(A=float * float(A=float * float(A=0.5, B=0.5), B=@n2))
   > Set Crown_Float_03(Crown_Float_03:Vector=MakeVector(X=float * float(A=@n10, B=0.7), Y=float * float(A=@n9, B=0.7), Z=float * float(A=@n11, B=0.7))) -> (Output_Get:Vector)
   n12 = Set pawnVelocitySpeed(execute=@n1.then_1, pawnVelocitySpeed:double=Vector Length(A=Get PawnVelocity())) -> (Output_Get:double)
   > n13 = Set pawnVelocitySpeedCutoff(pawnVelocitySpeedCutoff:double=SelectFloat(A=0, B=1, bPickA="float > float"(A=@n12, B=1200) {is=title K2Node_CallFunction})) -> (Output_Get:double)
   > Set SpringAlpha(SpringAlpha:double=float * float(A=0.01, B=@n13) {is=title K2Node_PromotableOperator}) -> (Output_Get:double)
   > n14 = Gravity Override State Function DEPRECATED(CharacterPartAnimInstance=Self-Reference(), GameTimeInSeconds=GetGameTimeInSeconds(), GravityOverrideParams=Get Gravity Override Params())
   > Set GravityOverride(GravityOverride=@n14)
   TryGetPawnOwner()
  ]}
]
variables [
 Crown_Float_01 {type=Vector}
 Crown_Float_02 {type=Vector}
 Crown_Float_03 {type=Vector}
 Gravity Override Params {type=GravityOverrideParamsStruct}
 GravityOverride {type=Vector}
 pawnVelocitySpeed {type=number}
 pawnVelocitySpeedCutoff {type=number}
 SpringAlpha {type=number}
]
